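; Quest ability for the Season 16 Battle Pass tower launch quest that reports when the player uses a launcher.
; On activation it keeps listening for gameplay effects the owner applies to targets, filtered by launch source and pawn tag requirements.
; Each match calls PlayerUsedLauncher, which sends a complex custom stat event with the target actor and target tags to the owning player's quest manager, counting one.
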(event Ability|EventActivateAbility
  (|CallParentFunction)
  (bind (_asynctaskproxy _target) (Ability|Tasks|WaitGameplayEffectAppliedToTarget :SourceTagRequirements (Variables|Default|GetTC_LaunchSourceTag) :TargetTagRequirements (Variables|Default|GetTC_PawnTag) :TriggerOnce false :ListenForPeriodicEffects false)
    (:OnApplied
      (CallFunction|PlayerUsedLauncher _target))))

(event Ability|EventFailedToActivatePassiveAbility
  (params (ActorInfo :type Gameplay%20Ability%20Actor%20Info%20Structure)))

(event Custom|PlayerUsedLauncher
  (params (TargetActor :type Actor%20Object%20Reference%20%28by%20ref%29))
  (Quest|SendComplexCustomStatEvent (Variables|GAVars|Data|GetOwningPlayerQuestManager) TargetActor "(GameplayTags=)" (Variables|Default|GetTC_TargetTags) 1))
